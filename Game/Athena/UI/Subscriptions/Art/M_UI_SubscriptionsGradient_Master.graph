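# Translucent UI master material for Subscriptions gradient backgrounds, offer images and masks.
# A radial gradient blends two or three colors, with optional screen-space UVs.
# A perspective pattern tile, broken up by three panning noise layers, tints it with HSV-shifted copies of those colors and color-dodges an angled shine.
# Static switches select character mode, fading the offer image into the gradient using its alpha, or a rotatable, offsettable hard-edged alpha mask.
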Material /Game/Athena/UI/Subscriptions/Art/M_UI_SubscriptionsGradient_Master
graph {
 properties {
  bUsedWithStaticMesh=false
  blendMode=BLEND_Translucent
  materialDomain=MD_UI
  pixelDepthOffsetMode=PDOM_Legacy
  usedShadingModels=Default Lit
 }
 nodes [
  n1 = ScalarParameter() {defaultValue=0.55, parameterName=BG_RadiusX}
  AppendVector(A=@n1, B=ScalarParameter() {defaultValue=0.55, parameterName=BG_RadiusY})
  MaterialFunctionCall() {materialFunction=/FortUILibrary/Foundation/Art_ToolKit/Tools/MF_UI_SquareToScreenSpace}
  MaterialFunctionCall() {materialFunction=&ScreenAlignedUVs}
  MaterialFunctionCall() {materialFunction=&ScreenResolution}
  n2 = ScalarParameter() {group=Mask, parameterName=Mask_Rotation_Start}
  n3 = ScalarParameter() {group=Mask, parameterName=Mask_Rotation_Stop}
  n4 = ScalarParameter() {defaultValue=1, group=Mask, parameterName=Animate_Mask}
  n5 = TextureCoordinate()
  n6 = OneMinus(None=@n5)
  n7 = ScalarParameter() {defaultValue=1, group=Mask, parameterName=Flip_Mask}
  n8 = ScalarParameter() {defaultValue=-0.5, group=Mask, parameterName=Offset_Mask_X_Start}
  n9 = ScalarParameter() {defaultValue=-0.5, group=Mask, parameterName=Offset_Mask_X_Stop}
  n10 = ScalarParameter() {defaultValue=1, group=Mask, parameterName=Animate_Mask}
  n11 = ScalarParameter() {group=Mask, parameterName=Offset_Mask_Y_Start}
  n12 = ScalarParameter() {group=Mask, parameterName=Offset_Mask_Y_Stop}
  n13 = ScalarParameter() {defaultValue=1, group=Mask, parameterName=Animate_Mask}
  n14 = Add(
   A=MaterialFunctionCall!(Rotation_Fraction=LinearInterpolate(A=@n2, B=@n3, Alpha=@n4), TexCoord=LinearInterpolate(A=@n5, B=@n6, Alpha=@n7)) {materialFunction=/FortUILibrary/Foundation/Art_ToolKit/AngledBoxes/MF_UI_Rotator_Fractions}.Result
   B=AppendVector(A=LinearInterpolate(A=@n8, B=@n9, Alpha=@n10), B=LinearInterpolate(A=@n11, B=@n12, Alpha=@n13))
  )
  n15 = Saturate(None=Multiply(
    A=StaticComponentMaskParameter(None=@n14) {defaultR=true, group=Mask, parameterName=Mask_Orientation}
    B=ScalarParameter() {defaultValue=128, group=Mask, parameterName=Mask_Edge_Hardness}
   ))
  n16 = Constant() {r=1}
  n17 = ScalarParameter() {defaultValue=1, group=Render, parameterName=Image_Scale}
  n18 = AppendVector(A=Constant() {r=0.5}, B=Constant() {r=1})
  n19 = Add(A=Divide(A=TextureCoordinate(), B=@n17), B=@n18)
  n20 = Divide(A=@n18, B=@n17) {desc=Scale the center offset by the scaling factor}
  n21 = ScalarParameter() {group=Render, parameterName=TranslateImage_X}
  n22 = ScalarParameter() {group=Render, parameterName=TranslateImage_Y}
  n23 = TextureSample(
   UVs=Add(A=Subtract(A=@n19, B=@n20) {desc=Re offset to maintain center}, B=AppendVector(A=@n21, B=@n22))
   Tex=TextureObjectParameter() {group=Render, parameterName=OfferImage, texture=/Game/Global/Textures/Default/Blanks/T_Blank_White_ZeroAlpha_sRGBColor}
  ) {samplerSource=SSM_Clamp_WorldGroupSettings}
  Multiply(
   A=StaticSwitchParameter(True=@n23.A, False=StaticSwitchParameter(True=@n15, False=@n16) {group=Global, parameterName=IsMask}) {group=Global, parameterName=IsCharacter}
   B=ScalarParameter() {defaultValue=1, group=Global, parameterName=Overall_Alpha}
  )
  n24 = TextureCoordinate()
  n25 = ScalarParameter() {defaultValue=100, group=Render, parameterName=FallOffColor_Postion_Start}
  n26 = ScalarParameter() {defaultValue=25, group=Render, parameterName=FallOffColor_Postion_Stop}
  n27 = ScalarParameter() {defaultValue=1, group=Render, parameterName=Animate_Fill_Color}
  n28 = LinearInterpolate(Alpha=Multiply(A=LinearInterpolate(A=@n25, B=@n26, Alpha=@n27)) {constB=0.01}) {constA=1, constB=0}
  n29 = Add(
   A=LinearInterpolate(A=@n24, B=OneMinus(None=@n24), Alpha=ScalarParameter() {group=Mask, parameterName=Flip_Fade})
   B=AppendVector(A=Constant(), B=Multiply(A=@n28) {constB=-1})
  )
  n30 = Saturate(None=Power(Base=ComponentMask(None=@n29) {g=true}, Exp=ScalarParameter() {defaultValue=1.22, group=Render, parameterName=FallOffColor_DecayRate}))
  n31 = Multiply(A=LinearInterpolate(
    A=ScalarParameter() {defaultValue=40, group=Render, parameterName=FallOffColor_Fill_Percent_Start}
    B=ScalarParameter() {defaultValue=40, group=Render, parameterName=FallOffColor_Fill_Percent_Stop}
    Alpha=ScalarParameter() {defaultValue=1, group=Render, parameterName=Animate_Fill_Color}
   )) {constB=0.01}
  n32 = Constant() {r=-36}
  n33 = Constant() {r=1}
  n34 = Add(
   A=Rotator(Coordinate=TextureCoordinate(), Time=ScalarParameter() {defaultValue=0.175, group=Animate, parameterName=ShineAngle}) {speed=6.2832}
   B=ScalarParameter() {defaultValue=-1, group=Animate, parameterName=ShineSwipe}
  )
  n35 = ComponentMask(None=Sine(None=Multiply(A=@n34) {constB=0.5})) {g=true}
  n36 = Saturate(None=LinearInterpolate(
    A=LinearInterpolate(A=@n32, B=@n33, Alpha=@n35)
    B=@n35
    Alpha=ScalarParameter() {defaultValue=1, group=Treatment, parameterName=DividerLine_Thiccness}
   ))
  n37 = Constant() {r=1}
  n38 = MaterialFunctionCall() {materialFunction=&GetUserInterfaceUV}
  n39 = MaterialFunctionCall(Float2=@n38.Pixel Size) {materialFunction=&BreakOutFloat2Components}
  n40 = Divide(A=@n39.G, B=@n39.G)
  n41 = OneMinus(None=@n40)
  n42 = Add(A=Multiply(A=@n38.Normalized UV, B=AppendVector(A=@n37, B=@n40)), B=AppendVector(A=Constant(), B=Multiply(A=@n41) {constB=0.5}))
  n43 = ScalarParameter() {defaultValue=2, group=Pattern, parameterName=Scale}
  n44 = TextureSample(UVs=MaterialFunctionCall!(
    UV=Multiply(A=@n42, B=@n43)
    Texture
    FixedScaleFactor_X
    FixedScaleFactor_Y
    Persp_Y=ScalarParameter() {group=Pattern, parameterName=Persp_Y}
    Persp_X=ScalarParameter() {group=Pattern, parameterName=Persp_X}
    Slide_X=ScalarParameter() {group=Pattern, parameterName=Slide_X}
    Slide_Y=ScalarParameter() {group=Pattern, parameterName=Slide_Y}
    Flip_Y
    Flip_X
   ) {materialFunction=/Game/UI/Foundation/Materials/MF_UI_BasicPerspective}.UnclampedFinalUvs) {texture=&/T_UI_SubsPatternTile_256x}
  n45 = DotProduct(A=@n44.R, B=@n44.R)
  n46 = Saturate(None=Subtract(A=Add(A=@n44.R, B=@n44.R), B=Saturate(None=@n45)))
  n47 = Multiply(A=@n46) {constB=0.333}
  n48 = DotProduct(A=Saturate(None=Power(Base=@n36) {constExponent=6}), B=Saturate(None=Add(A=@n44.B, B=@n47)))
  n49 = ScalarParameter() {defaultValue=0.5, group=Pattern, parameterName=PatternDodgeAmount}
  n50 = MaterialFunctionCall!(TextureResolution=Constant2Vector() {g=256, r=256}) {materialFunction=/Engine/Functions/Engine_MaterialFunctions02/Texturing/ScreenAlignedPixelToPixelUVs}
  n51 = ScalarParameter() {defaultValue=0.125, group=Noise, parameterName=Noise_1_Scale}
  n52 = ScalarParameter() {defaultValue=-0.125, group=Noise, parameterName=Noise_1_SpeedX, sliderMin=-0.05}
  n53 = ScalarParameter() {defaultValue=0.0225, group=Noise, parameterName=Noise_1_SpeedY}
  n54 = Multiply(A=Time(), B=ScalarParameter() {defaultValue=1, group=Global, parameterName=GlobalTime_Dialation})
  n55 = TextureSample(UVs=Panner(Coordinate=Multiply(A=@n50.Result, B=@n51), Time=@n54, Speed=AppendVector(A=@n52, B=@n53))) {samplerType=SAMPLERTYPE_Grayscale, texture=&LowResBlurredNoise}
  n56 = OneMinus(None=@n55.R)
  n57 = ScalarParameter() {defaultValue=0.125, group=Noise, parameterName=Noise_2_Scale}
  n58 = ScalarParameter() {defaultValue=0.0483, group=Noise, parameterName=Noise_2_SpeedX, sliderMin=-0.05}
  n59 = ScalarParameter() {defaultValue=-0.0308, group=Noise, parameterName=Noise_2_SpeedY}
  n60 = TextureSample(UVs=Panner(Coordinate=Multiply(A=@n50.Result, B=@n57), Time=@n54, Speed=AppendVector(A=@n58, B=@n59))) {samplerType=SAMPLERTYPE_Grayscale, texture=&LowResBlurredNoise}
  n61 = Saturate(None=Multiply(A=DotProduct(A=@n55.R, B=@n60.R)) {constB=3})
  n62 = Multiply(A=Saturate(None=DotProduct(A=@n56, B=@n61))) {constB=2}
  n63 = ScalarParameter() {defaultValue=0.125, group=Noise, parameterName=Noise_3_Scale}
  n64 = ScalarParameter() {defaultValue=0.0125, group=Noise, parameterName=Noise_3_SpeedX, sliderMin=-0.05}
  n65 = ScalarParameter() {defaultValue=-0.0075, group=Noise, parameterName=Noise_3_SpeedY}
  n66 = Add(
   A=Multiply(A=@n61, B=ScalarParameter() {defaultValue=0.125, group=Noise, parameterName=Noise_Distortion_Amount})
   B=Panner(Coordinate=Multiply(A=@n50.Result, B=@n63), Time=@n54, Speed=AppendVector(A=@n64, B=@n65))
  )
  n67 = Saturate(None=DotProduct(
    A=Saturate(None=@n62)
    B=TextureSample(UVs=@n66) {samplerType=SAMPLERTYPE_Grayscale, texture=&LowResBlurredNoise}.R
   ))
  n68 = Saturate(None=Power(Base=Multiply(A=@n67) {constA=1.5, constB=2}) {constExponent=1.5})
  n69 = ScalarParameter() {defaultValue=1.5, group=Pattern, parameterName=Pattern_Noise_Lift}
  n70 = DotProduct(A=@n46, B=Saturate(None=Multiply(A=@n68, B=@n69)))
  n71 = Saturate(None=LinearInterpolate(A=@n46, B=Saturate(None=@n70), Alpha=ScalarParameter() {defaultValue=0.9, group=Pattern, parameterName=Pattern_Noise_Amount}))
  n72 = VectorParameter() {group=Color, parameterName=Color_B, defaultValue=<rgba 1 0 0 1>}
  n73 = VectorParameter() {group=Color, parameterName=Color_A, defaultValue=<rgba 0 1 0 1>}
  n74 = ScalarParameter() {defaultValue=1, parameterName=BG_MultiplyUV_X}
  n75 = ScalarParameter() {defaultValue=1, parameterName=BG_MultiplyUV_Y}
  n76 = MaterialFunctionCall() {materialFunction=&GetUserInterfaceUV}
  n77 = MaterialFunctionCall(Float2=@n76.Pixel Size) {materialFunction=&BreakOutFloat2Components}
  n78 = Divide(A=@n77.G, B=@n77.G)
  n79 = AppendVector(A=Constant() {r=1}, B=@n78)
  n80 = Constant()
  n81 = Multiply(A=OneMinus(None=@n78)) {constB=0.5}
  n82 = StaticSwitchParameter(
   True=Multiply(
    A=MaterialFunctionCall() {materialFunction=&MF_UI_SquareToScreenSpace}.Y_Priority
    B=AppendVector(A=@n74, B=@n75)
   )
   False=Add(A=Multiply(A=@n76.Normalized UV, B=@n79), B=AppendVector(A=@n80, B=@n81))
  ) {parameterName=UseScreenSpaceUVs}
  n83 = AppendVector(
   A=ScalarParameter() {defaultValue=0.5, parameterName=BG_X_Position, sliderMin=0.5}
   B=ScalarParameter() {defaultValue=0.5, parameterName=BG_Y_Position}
  )
  n84 = ScalarParameter() {defaultValue=2, parameterName=BG_Density}
  n85 = MaterialFunctionCall!(
   X=ComponentMask(None=MaterialFunctionCall(UVs=@n82, CenterPosition=@n83, Radius=@n1, Density=@n84) {materialFunction=&RadialGradientExponential}.RadialGradientExponential) {r=true}
   Tangent 0=ScalarParameter() {group=Color, parameterName=BottomTangent}
   Tangent 1=ScalarParameter() {defaultValue=3, group=Color, parameterName=TopTangent}
  ) {materialFunction=/Engine/Functions/Engine_MaterialFunctions02/Gradients/SmoothCurve}
  n86 = Saturate(None=Saturate(None=ComponentMask(None=@n85.Result) {r=true}))
  n87 = OneMinus(None=ComponentMask(None=@n86) {r=true})
  n88 = ScalarParameter() {defaultValue=1, group=Color, parameterName=FlipDireciton}
  n89 = Saturate(None=Power(Base=LinearInterpolate(A=@n86, B=@n87, Alpha=@n88), Exp=ScalarParameter() {defaultValue=1, group=Color, parameterName=AdjustRate}) {constExponent=1.22})
  n90 = VectorParameter() {group=Color, parameterName=Color_A, defaultValue=<rgba 0 1 0 1>}
  n91 = VectorParameter() {group=Color, parameterName=Color_B, defaultValue=<rgba 1 0 0 1>}
  n92 = VectorParameter() {group=Color, parameterName=Color_C, defaultValue=<rgba 0 0 1 1>}
  n93 = ScalarParameter() {group=Color, parameterName=OffsetGradient}
  n94 = Multiply(A=@n93) {constB=-1}
  n95 = ScalarParameter() {defaultValue=1, group=Color, parameterName=FlipDireciton}
  n96 = ComponentMask(None=Add(B=LinearInterpolate(A=@n93, B=@n94, Alpha=@n95))) {r=true}
  n97 = OneMinus(None=@n96)
  n98 = ScalarParameter() {defaultValue=1, group=Color, parameterName=FlipDireciton}
  n99 = Saturate(None=Power(Base=LinearInterpolate(A=@n96, B=@n97, Alpha=@n98), Exp=ScalarParameter() {defaultValue=1, group=Color, parameterName=AdjustRate}) {constExponent=1.22})
  n100 = LinearInterpolate(
   A=LinearInterpolate(
    A=LinearInterpolate(A=@n72.RGB, B=@n73.RGB, Alpha=@n89)
    B=MaterialFunctionCall(A=@n90.RGB, B=@n91.RGB, C=@n92.RGB, Alpha=@n99) {materialFunction=&Lerp_3Color}.Result
    Alpha=ScalarParameter() {group=Color, parameterName=Use3ColorGradient}
   )
   Alpha=ScalarParameter() {group=Color, parameterName=Animate_toWhite}
  )
  n101 = MaterialFunctionCall(Float3=MaterialFunctionCall!(RGB=@n72.RGB) {materialFunction=/Engine/Functions/Engine_MaterialFunctions02/Math/RGBtoHSV}."RGB to HSV (0-1)") {materialFunction=&BreakOutFloat3Components}
  n102 = MaterialFunctionCall(
   X=@n101.R
   Y=Add(A=@n101.G, B=ScalarParameter() {defaultValue=-0.1, group=Pattern, parameterName=Color_B_Pattern_Saturation})
   Z=Add(A=@n101.B, B=ScalarParameter() {defaultValue=0.1, group=Pattern, parameterName=Color_B_Pattern_Value}) {constB=0.1}
  ) {materialFunction=&MakeFloat3}
  n103 = Saturate(None=MaterialFunctionCall!(HSV_In=Saturate(None=@n102.Result)) {materialFunction=/Game/Athena/UI/Materials/MF_HSVtoRGB}.RGB_Out)
  n104 = MaterialFunctionCall(Float3=MaterialFunctionCall(RGB=@n73.RGB) {materialFunction=&RGBtoHSV}."RGB to HSV (0-1)") {materialFunction=&BreakOutFloat3Components}
  n105 = MaterialFunctionCall(
   X=@n104.R
   Y=Add(A=@n104.G, B=ScalarParameter() {defaultValue=-0.1, group=Pattern, parameterName=Color_A_Pattern_Saturation})
   Z=Add(A=@n104.B, B=ScalarParameter() {defaultValue=0.1, group=Pattern, parameterName=Color_A_Pattern_Value}) {constB=0.1}
  ) {materialFunction=&MakeFloat3}
  n106 = Saturate(None=MaterialFunctionCall(HSV_In=Saturate(None=@n105.Result)) {materialFunction=&MF_HSVtoRGB}.RGB_Out)
  n107 = MaterialFunctionCall(Float3=MaterialFunctionCall(RGB=@n90.RGB) {materialFunction=&RGBtoHSV}."RGB to HSV (0-1)") {materialFunction=&BreakOutFloat3Components}
  n108 = MaterialFunctionCall(
   X=@n107.R
   Y=Add(A=@n107.G, B=ScalarParameter() {defaultValue=-0.1, group=Pattern, parameterName=Color_A_Pattern_Saturation})
   Z=Add(A=@n107.B, B=ScalarParameter() {defaultValue=0.1, group=Pattern, parameterName=Color_A_Pattern_Value}) {constB=0.1}
  ) {materialFunction=&MakeFloat3}
  n109 = Saturate(None=MaterialFunctionCall(HSV_In=Saturate(None=@n108.Result)) {materialFunction=&MF_HSVtoRGB}.RGB_Out)
  n110 = MaterialFunctionCall(Float3=MaterialFunctionCall(RGB=@n91.RGB) {materialFunction=&RGBtoHSV}."RGB to HSV (0-1)") {materialFunction=&BreakOutFloat3Components}
  n111 = MaterialFunctionCall(
   X=@n110.R
   Y=Add(A=@n110.G, B=ScalarParameter() {defaultValue=-0.1, group=Pattern, parameterName=Color_B_Pattern_Saturation})
   Z=Add(A=@n110.B, B=ScalarParameter() {defaultValue=0.1, group=Pattern, parameterName=Color_B_Pattern_Value}) {constB=0.1}
  ) {materialFunction=&MakeFloat3}
  n112 = Saturate(None=MaterialFunctionCall(HSV_In=Saturate(None=@n111.Result)) {materialFunction=&MF_HSVtoRGB}.RGB_Out)
  n113 = MaterialFunctionCall(Float3=MaterialFunctionCall(RGB=@n92.RGB) {materialFunction=&RGBtoHSV}."RGB to HSV (0-1)") {materialFunction=&BreakOutFloat3Components}
  n114 = MaterialFunctionCall(
   X=@n113.R
   Y=Add(A=@n113.G, B=ScalarParameter() {defaultValue=-0.1, group=Pattern, parameterName=Color_C_Pattern_Saturation})
   Z=Add(A=@n113.B, B=ScalarParameter() {defaultValue=0.1, group=Pattern, parameterName=Color_C_Pattern_Value}) {constB=0.1}
  ) {materialFunction=&MakeFloat3}
  n115 = Saturate(None=MaterialFunctionCall(HSV_In=Saturate(None=@n114.Result)) {materialFunction=&MF_HSVtoRGB}.RGB_Out)
  n116 = LinearInterpolate(
   A=LinearInterpolate(
    A=LinearInterpolate(A=@n103, B=@n106, Alpha=@n89)
    B=MaterialFunctionCall(A=@n109, B=@n112, C=@n115, Alpha=@n99) {materialFunction=&Lerp_3Color}.Result
    Alpha=ScalarParameter() {group=Color, parameterName=Use3ColorGradient}
   )
   Alpha=ScalarParameter() {group=Color, parameterName=Animate_toWhite}
  )
  n117 = LinearInterpolate(
   A=@n100
   B=MaterialFunctionCall!(Blend=Multiply(A=@n48, B=@n49) {constB=0.5}, Base=LinearInterpolate(A=@n100, B=@n116, Alpha=@n71)) {materialFunction=/Engine/Functions/Engine_MaterialFunctions03/Blends/Blend_ColorDodge}.Result
   Alpha=ScalarParameter() {defaultValue=1, group=Global, parameterName=UsePattern}
  )
  StaticSwitchParameter(True=LinearInterpolate(A=@n23.RGB, B=@n117, Alpha=Multiply(A=@n30, B=@n31)), False=@n117) {group=Global, parameterName=IsCharacter}
 ]
}
