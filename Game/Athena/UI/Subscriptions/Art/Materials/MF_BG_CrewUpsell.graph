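# Background material function for the Fortnite Crew upsell UI, with a static bIsSubscribed switch selecting between two looks.
# The unsubscribed look layers a radial gradient, two rotated, panning layers of CMS bento tiles, a vignette and a circular overlay.
# Random UV sampling varies tile scroll timing, and a 16-tap offset average blurs the background tiles.
# A separate branch builds a focal-point gradient with animated noise and overlay tint.
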MaterialFunction /Game/Athena/UI/Subscriptions/Art/Materials/MF_BG_CrewUpsell
graph {nodes [
  n1 = VectorParameter() {group=Background_Gradient, parameterName=RGB1, defaultValue=<rgba 0.0322 0.0018 0.0885 1>}
  n2 = LinearInterpolate(
   A=Multiply(A=@n1.RGB, B=ScalarParameter() {defaultValue=0.25, group=Vignette, parameterName=Vignette_Edge_Darken}) {constB=0.25}
   B=@n1.RGB
   Alpha=NamedRerouteUsage()
  )
  n3 = Constant2Vector() {g=1, r=0.25}
  n4 = Multiply(A=NamedRerouteUsage(), B=NamedRerouteUsage())
  n5 = Multiply(A=Time(), B=NamedRerouteUsage())
  n6 = NamedRerouteUsage()
  n7 = TextureSample(UVs=Panner(Coordinate=Multiply(A=@n3, B=@n4), Time=Multiply(A=@n5, B=@n6)) {speedX=0.125}, Tex=NamedRerouteUsage()) {samplerSource=SSM_Wrap_WorldGroupSettings}
  n8 = Clamp(None=NamedRerouteUsage()) {minDefault=0.09}
  n9 = ScalarParameter() {defaultValue=1.75, group=Background_Gradient, parameterName=Background_Radial_Radius}
  n10 = Add(A=NamedRerouteUsage(), B=Add(A=Constant2Vector() {g=0.5, r=-0.35}, B=NamedRerouteUsage()))
  n11 = SmoothStep(Value=OneMinus(None=MaterialFunctionCall!(UVs=@n10, Radius=@n9) {materialFunction=/FortUILibrary/MaterialFunctions/Shapes/SDFs/MF_SDF_Circle}.SDF)) {constMax=4}
  n12 = NamedRerouteDeclaration(None=LinearInterpolate(
    A=VectorParameter() {group=Background_Gradient, parameterName=RGB1, defaultValue=<rgba 0.0322 0.0018 0.0885 1>}.RGB
    B=VectorParameter() {group=Background_Gradient, parameterName=RGB2, defaultValue=<rgba 0.156 0.2385 0.651 1>}.RGB
    Alpha=Saturate(None=@n11)
   )) {name=Background_Gradient, nodeColor=<rgba 0 0 0 1>}
  n13 = NamedRerouteUsage()
  n14 = NamedRerouteUsage()
  n15 = Multiply(
   A=Multiply(A=NamedRerouteUsage(), B=Clamp(None=@n14) {minDefault=0.5})
   B=ScalarParameter() {defaultValue=0.125, group=Tiles_Background, parameterName=BG_Tiles_Intensity}
  )
  n16 = LinearInterpolate(
   A=@n2
   B=LinearInterpolate(A=LinearInterpolate(A=@n12, B=@n13, Alpha=@n15), B=@n7.RGB, Alpha=Multiply(A=@n7.A, B=@n8))
   Alpha=NamedRerouteUsage()
  )
  n17 = NamedRerouteUsage()
  n18 = Add(
   A=NamedRerouteUsage()
   B=AppendVector(
    A=ScalarParameter() {group=Overlay, parameterName=Overlay_X_Offset}
    B=ScalarParameter() {defaultValue=0.08, group=Overlay, parameterName=Overlay_Y_Offset}
   )
  )
  n19 = ScalarParameter() {defaultValue=0.25, group=Overlay, parameterName=Overlay_Size}
  n20 = ScalarParameter() {defaultValue=0.6348, group=Overlay, parameterName=Overlay_Glow}
  n21 = Saturate(None=Multiply(
    A=MaterialFunctionCall(UVs=@n18, Size=@n19, GlowMax=@n20) {materialFunction=&MF_UI_Circle}.Fill
    B=ScalarParameter() {group=Overlay, parameterName=Overlay_Intensity}
   ))
  FunctionOutput(None=StaticSwitchParameter(True=NamedRerouteUsage(), False=LinearInterpolate(A=@n16, B=@n17, Alpha=@n21)) {group=Global, parameterName=bIsSubscribed}) {outputName="Switch Param (False)"}
  NamedRerouteDeclaration(None=Rotator(Coordinate=NamedRerouteUsage(), Time=ScalarParameter() {defaultValue=0.08, group=Global, parameterName=Tile_Angle}) {speed=6.2832}) {name=Screen_Angle, nodeColor=<rgba 0 0 0 1>}
  n22 = Multiply(A=NamedRerouteUsage(), B=AppendVector(A=Constant() {r=0.5}, B=NamedRerouteUsage()))
  n23 = AppendVector(
   A=ScalarParameter() {defaultValue=0.25, group=Tiles_Foreground, parameterName=FG_Time_Offset_X-Tiling}
   B=ScalarParameter() {defaultValue=1, group=Tiles_Foreground, parameterName=FG_Time_Offset_Y-Tiling}
  )
  n24 = ScalarParameter() {defaultValue=-0.5, group=Tiles_Foreground, parameterName=FG_Time_Offset_X-Offsets}
  n25 = Round(None=ScalarParameter() {defaultValue=4, group=Tiles_Foreground, parameterName=FG_Time_Offset_Y-Offsets})
  n26 = MaterialFunctionCall(Float2=Add(A=ScalarParameter() {defaultValue=0.0625, group=Tiles_Foreground, parameterName=FG_Time_Mask_Offset}, B=@n22)) {materialFunction=&BreakOutFloat2Components}
  n27 = ConstantBiasScale(None=Add(A=MaterialFunctionCall(Value=@n26.G) {materialFunction=&LinearSine}.Linear Sine) {constB=-0.5}) {bias=-0.125, scale=12}
  n28 = MaterialFunctionCall!(
   PrimaryTextureCoordinates=Add(A=Multiply(A=@n22, B=@n23), B=AppendVector(A=@n24, B=@n25))
   2ndOrderTextureCoordinates
   RandomSeed
   NoiseDensityBias=LinearInterpolate(
    A=ScalarParameter() {defaultValue=1, group=Tiles_Foreground, parameterName=FG_Time_Mask_Upper_Range}
    B=ScalarParameter() {defaultValue=0.125, group=Tiles_Foreground, parameterName=FG_Time_Mask_Lower_Range}
    Alpha=Saturate(None=@n27)
   ) {constA=1, constB=0.5}
   Bypass Noise Influence Crosstalk
  ) {materialFunction=/Game/Effects/Fort_Effects/Materials/_Functions/MF_RandomUVSample}
  NamedRerouteDeclaration(None=Multiply(
    A=ConstantBiasScale(None=@n28.NoiseWithRange) {bias=0.5, scale=-2}
    B=ScalarParameter() {defaultValue=0.05, group=Tiles_Foreground, parameterName=Tile_FG_Time_Factor}
   )) {name=FG_Time_Offset, nodeColor=<rgba 0 0 0 1>}
  n29 = Add(A=NamedRerouteUsage(), B=NamedRerouteUsage())
  n30 = ScalarParameter() {defaultValue=0.3, group=Global, parameterName=Radial_Mask_Radius}
  NamedRerouteDeclaration(None=Saturate(None=MaterialFunctionCall(UVs=@n29, Radius=@n30) {materialFunction=&MF_SDF_Circle}.SDF)) {name=SDF_Radial_Mask, nodeColor=<rgba 0 0 0 1>}
  n31 = ScalarParameter() {defaultValue=2.4, group=Vignette, parameterName=Vignette_Radial_Radius}
  n32 = SmoothStep(Value=OneMinus(None=MaterialFunctionCall(UVs=@n10, Radius=@n31) {materialFunction=&MF_SDF_Circle}.SDF)) {constMax=3.5}
  n33 = ScalarParameter() {defaultValue=3.5, group=Vignette, parameterName=Vignette_Power}
  NamedRerouteDeclaration(None=Saturate(None=Power(Base=@n32, Exp=@n33))) {name=Vignette_Mask, nodeColor=<rgba 0 0 0 1>}
  NamedRerouteDeclaration(None=AppendVector(
    A=ScalarParameter() {defaultValue=0.328, group=Background_Gradient, parameterName=Gradient_X_Offset}
    B=ScalarParameter() {group=Background_Gradient, parameterName=Gradient_Y_Offset}
   )) {name=Screen_Offset, nodeColor=<rgba 0 0 0 1>}
  n34 = MaterialFunctionCall!(Use Local UV, Texture Size, Scale Pivot Point) {materialFunction=/Game/Environments/FrontEnd/Blueprints/ItemPreview/VaultWorldTransitions/MaterialFunctions/MF_UI_VaultWorld_UV}
  NamedRerouteDeclaration(None=@n34.Scaled to Fit) {name=UV, nodeColor=<rgba 0 0 0 1>}
  NamedRerouteDeclaration(None=@n34.Scale Y) {name=UV-Scale-Y, nodeColor=<rgba 0 0 0 1>}
  n35 = NamedRerouteUsage()
  n36 = ScalarParameter() {group=Gradient, parameterName=GradientStart}
  n37 = ScalarParameter() {defaultValue=1, group=Gradient, parameterName=GradientEnd}
  NamedRerouteDeclaration(None=LinearInterpolate(
    A=NamedRerouteUsage()
    B=NamedRerouteUsage()
    Alpha=MaterialFunctionCall(UV=@n35, GradientStart=@n36, GradientEnd=@n37) {materialFunction=&MF_UI_RadialGradient}.Smooth
   )) {name=BackgroundGradient, nodeColor=<rgba 0 0 0 1>}
  NamedRerouteDeclaration(None=ScalarParameter() {defaultValue=1, parameterName=AnimateBackground, sliderMax=1}) {name=AnimateBackground, nodeColor=<rgba 0 0 0 1>}
  NamedRerouteDeclaration(None=MaterialFunctionCall(UVs=NamedRerouteUsage(), GradientStart=Constant(), GradientEnd=Constant() {r=0.5}, RotationAngle=Constant() {r=0.25}) {materialFunction=&MF_UI_LinearGradient}.Result) {name=Overlay, nodeColor=<rgba 0 0 0 1>}
  NamedRerouteDeclaration(None=ScalarParameter() {defaultValue=1.675, group=Tiles_Foreground, parameterName=Tile-Scale-Foreground}) {name=FG_Tile_Scale, nodeColor=<rgba 0 0 0 1>}
  NamedRerouteDeclaration(None=VectorParameter() {group=Color, parameterName=Color1, defaultValue=<rgba 0.0232 0.0056 0.0865 1>}.RGB) {name=BackgroundGradient_Color1, nodeColor=<rgba 0 0 0 1>}
  NamedRerouteDeclaration(None=VectorParameter() {group=Color, parameterName=Color2, defaultValue=<rgba 0.2664 0.4397 0.7529 1>}.RGB) {name=BackgroundGradient_Color2, nodeColor=<rgba 0 0 0 1>}
  NamedRerouteDeclaration(None=VectorParameter() {group=Color, parameterName=Color4, defaultValue=<rgba 0.0097 0.0021 0.0284 1>}.RGB) {name=Overlay_Color, nodeColor=<rgba 0 0 0 1>}
  n38 = Add(
   A=NamedRerouteUsage()
   B=AppendVector(
    A=ScalarParameter() {defaultValue=-0.8, group=Gradient, parameterName=GradientXPan}
    B=ScalarParameter() {defaultValue=0.5, group=Gradient, parameterName=GradientYPan}
   )
  )
  n39 = ScalarParameter() {defaultValue=0.05, group=Gradient, parameterName=GradientRotation, sliderMax=0.5}
  n40 = AppendVector(
   A=ScalarParameter() {defaultValue=1.7, group=Gradient, parameterName=GradientXScale}
   B=ScalarParameter() {defaultValue=1, group=Gradient, parameterName=GradientYScale}
  )
  n41 = ScalarParameter() {defaultValue=2, group=Gradient, parameterName=GradientScale}
  NamedRerouteDeclaration(None=MaterialFunctionCall(
    UVs=MaterialFunctionCall(UVs=@n38, "Rotation Angle (0-1)"=@n39) {materialFunction=&CustomRotator}.Rotated Values
    Texture Scale=Multiply(A=@n40, B=@n41)
   ) {materialFunction=&ScaleUVsByCenter}.UVs) {name=UV_FocalPoint, nodeColor=<rgba 0 0 0 1>}
  n42 = Multiply(A=Time(), B=ScalarParameter() {defaultValue=0.5, parameterName=NoiseTime}) {constB=0.5}
  NamedRerouteDeclaration(None=Min(B=MaterialFunctionCall!(UV, Seed=@n42) {materialFunction=&MF_UI_Random}.Result) {constA=0.5}) {name=Noise, nodeColor=<rgba 0 0 0 1>}
  n43 = NamedRerouteUsage()
  n44 = MaterialFunctionCall(Base=@n43, Blend=NamedRerouteUsage()) {materialFunction=&Blend_Overlay}
  n45 = ScalarParameter() {defaultValue=0.15, parameterName=NoiseAlpha}
  n46 = NamedRerouteUsage()
  n47 = NamedRerouteDeclaration(None=OneMinus(None=NamedRerouteUsage())) {name=Overlay_Mask, nodeColor=<rgba 0 0 0 1>}
  n48 = ScalarParameter() {defaultValue=0.5, group=Overlay, parameterName=Overlay_Strength}
  NamedRerouteDeclaration(None=LinearInterpolate(
    A=LinearInterpolate(A=@n43, B=@n44.Result, Alpha=@n45) {constAlpha=0.3}
    B=LinearInterpolate(B=@n46, Alpha=@n47)
    Alpha=Multiply(A=@n47, B=@n48) {constB=0.6}
   )) {name=GradientOutput, nodeColor=<rgba 0 0 0 1>}
  n49 = TextureObjectParameter() {
   parameterName=CMS_Crew_Bento
   samplerSource=SSM_Wrap_WorldGroupSettings
   texture=/Game/Athena/UI/Subscriptions/Art/Textures/T_CrewUpsell_Bento
  }
  NamedRerouteDeclaration(None=FunctionInput(Preview=@n49) {bUsePreviewValueAsDefault=true, inputName=CMSTexture, inputType=FunctionInput_Texture2D}) {name=Texture Object, nodeColor=<rgba 0 0 0 1>}
  NamedRerouteDeclaration(None=Multiply(A=ScalarParameter() {defaultValue=4, group=BlurSampling, parameterName=Factor}, B=Constant2Vector() {g=0.004, r=0.001})) {name=Factor_Scale_Compensation, nodeColor=<rgba 0 0 0 1>}
  n50 = Multiply(A=NamedRerouteUsage(), B=AppendVector(A=Constant() {r=0.5}, B=NamedRerouteUsage()))
  n51 = AppendVector(
   A=ScalarParameter() {defaultValue=0.25, group=Tiles_Background, parameterName=BG_Time_Offset_X-Tiling}
   B=ScalarParameter() {defaultValue=1, group=Tiles_Background, parameterName=BG_Time_Offset_Y-Tiling}
  )
  n52 = ScalarParameter() {defaultValue=-0.5, group=Tiles_Background, parameterName=BG_Time_Offset_X-Offsets}
  n53 = Round(None=ScalarParameter() {defaultValue=3, group=Tiles_Background, parameterName=BG_Time_Offset_Y-Offsets})
  n54 = MaterialFunctionCall(Float2=Add(A=ScalarParameter() {defaultValue=0.0625, group=Tiles_Background, parameterName=BG_Time_Mask_Offset}, B=@n50)) {materialFunction=&BreakOutFloat2Components}
  n55 = ConstantBiasScale(None=Add(A=MaterialFunctionCall(Value=@n54.G) {materialFunction=&LinearSine}.Linear Sine) {constB=-0.5}) {bias=-0.125, scale=12}
  n56 = MaterialFunctionCall(
   PrimaryTextureCoordinates=Add(A=Multiply(A=@n50, B=@n51), B=AppendVector(A=@n52, B=@n53))
   NoiseDensityBias=LinearInterpolate(
    A=ScalarParameter() {defaultValue=1, group=Tiles_Background, parameterName=BG_Time_Mask_Upper_Range}
    B=ScalarParameter() {defaultValue=1.25, group=Tiles_Background, parameterName=BG_Time_Mask_Lower_Range}
    Alpha=Saturate(None=@n55)
   ) {constA=1, constB=0.5}
  ) {materialFunction=&MF_RandomUVSample}
  NamedRerouteDeclaration(None=Multiply(
    A=ConstantBiasScale(None=@n56.NoiseWithRange) {bias=0.5, scale=-2}
    B=ScalarParameter() {defaultValue=0.1, group=Tiles_Background, parameterName=Tile_BG_Time_Factor}
   )) {name=BG_Time_Offset, nodeColor=<rgba 0 0 0 1>}
  NamedRerouteDeclaration(None=ScalarParameter() {defaultValue=1.8, group=Tiles_Background, parameterName=Tile-Scale-Background}) {name=BG_Tile_Scale, nodeColor=<rgba 0 0 0 1>}
  n57 = Constant2Vector() {g=1, r=0.25}
  n58 = Multiply(A=NamedRerouteUsage(), B=NamedRerouteUsage())
  n59 = Multiply(A=Time(), B=NamedRerouteUsage())
  n60 = NamedRerouteUsage()
  NamedRerouteDeclaration(None=Panner(Coordinate=Multiply(A=@n57, B=@n58), Time=Multiply(A=@n59, B=@n60)) {speedX=0.125}) {name=BG_UV_forSampling, nodeColor=<rgba 0 0 0 1>}
  n61 = NamedRerouteUsage()
  n62 = Constant2Vector() {g=-1, r=1}
  n63 = TextureSample(UVs=Add(A=NamedRerouteUsage(), B=Multiply(A=@n61, B=@n62)), Tex=NamedRerouteUsage()) {samplerSource=SSM_Wrap_WorldGroupSettings}
  n64 = NamedRerouteUsage()
  n65 = Constant2Vector() {g=-1, r=-1}
  n66 = TextureSample(UVs=Add(A=NamedRerouteUsage(), B=Multiply(A=@n64, B=@n65)), Tex=NamedRerouteUsage()) {samplerSource=SSM_Wrap_WorldGroupSettings}
  n67 = Add(A=@n66.RGB, B=@n63.RGB)
  n68 = NamedRerouteUsage()
  n69 = Constant2Vector() {g=1, r=-1}
  n70 = TextureSample(UVs=Add(A=NamedRerouteUsage(), B=Multiply(A=@n68, B=@n69)), Tex=NamedRerouteUsage()) {samplerSource=SSM_Wrap_WorldGroupSettings}
  n71 = NamedRerouteUsage()
  n72 = Constant2Vector() {g=1, r=1}
  n73 = TextureSample(UVs=Add(A=NamedRerouteUsage(), B=Multiply(A=@n71, B=@n72)), Tex=NamedRerouteUsage()) {samplerSource=SSM_Wrap_WorldGroupSettings}
  n74 = NamedRerouteUsage()
  n75 = Constant2Vector() {g=-0.5, r=0.5}
  n76 = TextureSample(UVs=Add(A=NamedRerouteUsage(), B=Multiply(A=@n74, B=@n75)), Tex=NamedRerouteUsage()) {samplerSource=SSM_Wrap_WorldGroupSettings}
  n77 = NamedRerouteUsage()
  n78 = Constant2Vector() {g=-0.5, r=-0.5}
  n79 = TextureSample(UVs=Add(A=NamedRerouteUsage(), B=Multiply(A=@n77, B=@n78)), Tex=NamedRerouteUsage()) {samplerSource=SSM_Wrap_WorldGroupSettings}
  n80 = Add(A=@n79.RGB, B=@n76.RGB)
  n81 = NamedRerouteUsage()
  n82 = Constant2Vector() {g=0.5, r=-0.5}
  n83 = TextureSample(UVs=Add(A=NamedRerouteUsage(), B=Multiply(A=@n81, B=@n82)), Tex=NamedRerouteUsage()) {samplerSource=SSM_Wrap_WorldGroupSettings}
  n84 = NamedRerouteUsage()
  n85 = Constant2Vector() {g=0.5, r=0.5}
  n86 = TextureSample(UVs=Add(A=NamedRerouteUsage(), B=Multiply(A=@n84, B=@n85)), Tex=NamedRerouteUsage()) {samplerSource=SSM_Wrap_WorldGroupSettings}
  n87 = Add(A=Add(A=Add(A=@n67, B=@n70.RGB), B=@n73.RGB), B=Add(A=Add(A=@n80, B=@n83.RGB), B=@n86.RGB))
  n88 = NamedRerouteUsage()
  n89 = Constant2Vector() {r=0.5}
  n90 = TextureSample(UVs=Add(A=NamedRerouteUsage(), B=Multiply(A=@n88, B=@n89)), Tex=NamedRerouteUsage()) {samplerSource=SSM_Wrap_WorldGroupSettings}
  n91 = NamedRerouteUsage()
  n92 = Constant2Vector() {r=-0.5}
  n93 = TextureSample(UVs=Add(A=NamedRerouteUsage(), B=Multiply(A=@n91, B=@n92)), Tex=NamedRerouteUsage()) {samplerSource=SSM_Wrap_WorldGroupSettings}
  n94 = Add(A=@n93.RGB, B=@n90.RGB)
  n95 = NamedRerouteUsage()
  n96 = Constant2Vector() {g=0.5}
  n97 = TextureSample(UVs=Add(A=NamedRerouteUsage(), B=Multiply(A=@n95, B=@n96)), Tex=NamedRerouteUsage()) {samplerSource=SSM_Wrap_WorldGroupSettings}
  n98 = NamedRerouteUsage()
  n99 = Constant2Vector() {g=0.5}
  n100 = TextureSample(UVs=Add(A=NamedRerouteUsage(), B=Multiply(A=@n98, B=@n99)), Tex=NamedRerouteUsage()) {samplerSource=SSM_Wrap_WorldGroupSettings}
  n101 = NamedRerouteUsage()
  n102 = Constant2Vector() {r=1}
  n103 = TextureSample(UVs=Add(A=NamedRerouteUsage(), B=Multiply(A=@n101, B=@n102)), Tex=NamedRerouteUsage()) {samplerSource=SSM_Wrap_WorldGroupSettings}
  n104 = NamedRerouteUsage()
  n105 = Constant2Vector() {r=-1}
  n106 = TextureSample(UVs=Add(A=NamedRerouteUsage(), B=Multiply(A=@n104, B=@n105)), Tex=NamedRerouteUsage()) {samplerSource=SSM_Wrap_WorldGroupSettings}
  n107 = Add(A=@n106.RGB, B=@n103.RGB)
  n108 = NamedRerouteUsage()
  n109 = Constant2Vector() {g=1}
  n110 = TextureSample(UVs=Add(A=NamedRerouteUsage(), B=Multiply(A=@n108, B=@n109)), Tex=NamedRerouteUsage()) {samplerSource=SSM_Wrap_WorldGroupSettings}
  n111 = NamedRerouteUsage()
  n112 = Constant2Vector() {g=1}
  n113 = TextureSample(UVs=Add(A=NamedRerouteUsage(), B=Multiply(A=@n111, B=@n112)), Tex=NamedRerouteUsage()) {samplerSource=SSM_Wrap_WorldGroupSettings}
  n114 = Add(A=Add(A=Add(A=@n94, B=@n97.RGB), B=@n100.RGB), B=Add(A=Add(A=@n107, B=@n110.RGB), B=@n113.RGB))
  NamedRerouteDeclaration(None=Multiply(A=Add(A=@n87, B=@n114)) {constB=0.0625}) {name=BG_Blured_RGB_Output, nodeColor=<rgba 0 0 0 1>}
  n115 = Add(A=@n63.A, B=@n66.A)
  n116 = Add(A=@n76.A, B=@n79.A)
  n117 = Add(A=Add(A=Add(A=@n115, B=@n70.A), B=@n73.A), B=Add(A=Add(A=@n116, B=@n83.A), B=@n86.A))
  n118 = Add(A=@n90.A, B=@n93.A)
  n119 = Add(A=@n103.A, B=@n106.A)
  n120 = Add(A=Add(A=Add(A=@n118, B=@n97.A), B=@n100.A), B=Add(A=Add(A=@n119, B=@n110.A), B=@n113.A))
  NamedRerouteDeclaration(None=Multiply(A=Add(A=@n117, B=@n120)) {constB=0.0625}) {name=BG_Blured_Alpha_Output, nodeColor=<rgba 0 0 0 1>}
  Reroute(None=LinearInterpolate(A=@n2, B=NamedRerouteUsage(), Alpha=NamedRerouteUsage()))
 ]}
